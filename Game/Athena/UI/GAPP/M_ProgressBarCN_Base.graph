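# Translucent unlit UI material for a progress bar that changes fill color as Progress rises.
# A linear U gradient compared against Progress separates the Unfilled Color from a fill chosen by nested Ifs in quarter steps, from ColorB through red and gold to teal.
# A rotated, tiled gradient texture through SmoothStep builds a segment alpha mask.
# The UseSheenPulse static switch blends in SheenColor through a shine texture offset by Position.
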Material /Game/Athena/UI/GAPP/M_ProgressBarCN_Base
graph {
 properties {
  bUsedWithStaticMesh=false
  blendMode=BLEND_Translucent
  materialDomain=MD_UI
  numCustomizedUVs=2
  pixelDepthOffsetMode=PDOM_Legacy
  shadingModel=MSM_Unlit
  usedShadingModels=Unlit
 }
 nodes [
  n1 = ScalarParameter() {defaultValue=0.75, parameterName=Progress}
  Ceil(None=Divide(A=@n1, B=Constant() {r=0.25}))
  n2 = MaterialFunctionCall(UV Channel=TextureCoordinate()) {materialFunction=&LinearGradient}
  n3 = VectorParameter() {parameterName=Unfilled Color, defaultValue=<rgba 0 0 0 0.25>}
  n4 = VectorParameter() {parameterName=ColorB, defaultValue=<rgba 0.6795 0.0272 0.0272 1>}
  n5 = LinearInterpolate(A=VectorParameter() {parameterName=ColorA, defaultValue=<rgba 1 0 0 1>}.A, B=@n4.A, Alpha=@n2.UGradient)
  If(A=@n2.UGradient, B=@n1, "A > B"=@n3.A, "A == B"=@n5, "A < B"=@n5)
  MaterialFunctionCall(
   UVs=MaterialFunctionCall() {materialFunction=&GetUserInterfaceUV}.9-Slice UV
   "Rotation Angle (0-1)"=Constant() {r=0.02}
  ) {materialFunction=&CustomRotator}
  Multiply()
  n6 = TextureCoordinate()
  n7 = ScalarParameter() {defaultValue=0.25, parameterName=SegmentAngle}
  n8 = Constant()
  n9 = ScalarParameter() {defaultValue=0.0004, parameterName=SegmentsOffset}
  n10 = ScalarParameter() {defaultValue=1, parameterName=Segments}
  n11 = Multiply(
   A=Add(
    A=MaterialFunctionCall(UVs=@n6, "Rotation Angle (0-1)"=@n7) {materialFunction=&CustomRotator}.Rotated Values
    B=AppendVector(A=@n8, B=@n9)
   ) {constB=0}
   B=AppendVector(A=Constant() {desc=stretch vertical to eliminate artifacts, r=0.01}, B=@n10)
  )
  n12 = Multiply(A=@n10, B=ScalarParameter() {defaultValue=0.0005, parameterName=SegmentEdge})
  n13 = Multiply(A=@n10, B=ScalarParameter() {defaultValue=0, parameterName=SegmentEdgeSoftness})
  Multiply(
   A=MaterialFunctionCall(
    Alpha=TextureSample(UVs=@n11) {texture=/Game/Global/Textures/Default/Masks_and_Gradients/T_UI_ReticleSegmentGradient}.R
    Min=@n12
    Max=Add(A=@n12, B=@n13)
   ) {materialFunction=&SmoothStep}.Result
   B=@n5
  )
  n14 = ScalarParameter() {defaultValue=0.75, parameterName=Progress, sliderMax=1}
  n15 = Constant3Vector() {constant=<rgba 0.1052 0.67 0.5358 0>}
  n16 = Constant3Vector() {constant=<rgba 0.807 0.5395 0.1221 0>}
  n17 = Constant() {r=0.25}
  n18 = Constant3Vector() {constant=<rgba 0.6514 0.1248 0.0723 1>}
  n19 = If(
   A=@n14
   B=Constant() {r=0.75}
   "A > B"=@n15
   "A == B"=@n15
   "A < B"=If(A=@n14, B=Constant() {r=0.5}, "A > B"=@n16, "A == B"=@n16, "A < B"=If(A=@n14, B=@n17, "A > B"=@n18, "A == B"=@n18, "A < B"=@n4.RGB))
  )
  n20 = If(A=@n2.UGradient, B=@n1, "A > B"=@n3.RGB, "A == B"=@n19, "A < B"=@n19)
  n21 = LinearInterpolate(Alpha=ScalarParameter() {parameterName=Position}) {constA=1, constAlpha=0.2, constB=-1}
  n22 = Add(A=TextureCoordinate() {coordinateIndex=1}, B=AppendVector(A=@n21, B=@n21))
  StaticSwitchParameter(
   True=LinearInterpolate(
    A=@n20
    B=VectorParameter() {parameterName=SheenColor, defaultValue=<rgba 1 1 1 1>}.RGB
    Alpha=TextureSample(UVs=@n22) {samplerSource=SSM_Clamp_WorldGroupSettings, texture=/Game/UI/Foundation/Textures/Pattern/T-UI-Shine}.R
   )
   False=@n20
  ) {parameterName=UseSheenPulse}
  TextureCoordinate()
 ]
}
