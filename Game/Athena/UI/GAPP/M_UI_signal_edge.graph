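# Translucent UI material that draws a tintable signal edge graphic with an optional pulsing glow state.
# If nodes switch from TextureA to the glow texture when the red channel of Glow equals 1, with color scaled by EdgeColor and brightness in both states.
# Glow alpha pulses on a clamped time-driven sine and ignores Opacity, which scales only the normal state.
# Glow color samples a fixed texture, while glow alpha uses the Param texture parameter.
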Material /Game/Athena/UI/GAPP/M_UI_signal_edge
graph {
 properties {
  bUsedWithStaticMesh=false
  blendMode=BLEND_Translucent
  materialDomain=MD_UI
  pixelDepthOffsetMode=PDOM_Legacy
  usedShadingModels=Default Lit
 }
 nodes [
  n1 = VectorParameter() {parameterName=Glow, defaultValue=<rgba 0 0 0 1>}
  n2 = Constant3Vector() {constant=<rgba 1 1 1 0>}
  n3 = TextureSampleParameter2D() {parameterName=TextureA, texture=&/T_UI_signal_edge}
  n4 = Multiply(A=@n3.A, B=ScalarParameter() {defaultValue=1, parameterName=Opacity})
  n5 = Multiply(A=Time(), B=Constant() {r=1})
  n6 = Clamp(None=Add(A=Sine(None=@n5), B=Constant() {r=0.05}))
  If(
   A=ComponentMask(None=@n1.RGB) {r=true}
   B=ComponentMask(None=@n2) {r=true}
   "A > B"=@n4
   "A == B"=Multiply(
    A=TextureSampleParameter2D() {parameterName=Param, texture=&/T_UI_signal_edgeGlow}.A
    B=LinearInterpolate(Alpha=@n6) {constA=0.5, constB=1.5}
   )
   "A < B"=@n4
  )
  n7 = Multiply(
   A=Multiply(A=@n3.RGB, B=VectorParameter() {parameterName=EdgeColor, defaultValue=<rgba 1 1 1 1>}.RGB)
   B=ScalarParameter() {defaultValue=1, parameterName=brightness}
  )
  n8 = TextureSample() {samplerSource=SSM_Wrap_WorldGroupSettings, texture=&T_UI_signal_edgeGlow}
  n9 = VectorParameter() {parameterName=EdgeColor, defaultValue=<rgba 1 1 1 1>}
  If(
   A=ComponentMask(None=@n1.RGB) {r=true}
   B=ComponentMask(None=@n2) {r=true}
   "A > B"=@n7
   "A == B"=Multiply(A=Multiply(A=@n8.RGB, B=@n9.RGB), B=ScalarParameter() {defaultValue=1, parameterName=brightness})
   "A < B"=@n7
  )
 ]
}
