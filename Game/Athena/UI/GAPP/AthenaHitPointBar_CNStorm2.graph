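# CN storm variant of the Athena hit point bar, shown only in Last Man Standing CN games, with a signal-styled multiplier label.
# Bar type sets initial styling; value changes drive fill and delta bars.
# Quarter thresholds of the current percentage select multiplier text (x1.0 near full) and colors, while last-versus-current comparisons drive a pulsing warning edge and glow through dynamic material parameters.
# DBNO reddens health bars and hides shield or signal bars.
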WidgetBlueprint /Game/Athena/UI/GAPP/AthenaHitPointBar_CNStorm2
animations [warning {
  binding_count=1
  length_seconds=0.55
  object=&self:warning
  sequence {bindings [BorderEdge3 {
     id=#1
     possessed_class=Image
     tracks [
      {
       type=MovieScene2DTransformTrack
       sections [{
         type=MovieScene2DTransformSection
         end_frame=11
         start_frame=0
         channels [
          Translation.X {type=MovieSceneScriptingFloatChannel, const [0,1]}
          Translation.Y {type=MovieSceneScriptingFloatChannel, const [0,1]}
          Angle {type=MovieSceneScriptingFloatChannel, const [0,1]}
          Scale.X {type=MovieSceneScriptingFloatChannel, keys [[1,1],[10,1.06],[11,1]]}
          Scale.Y {type=MovieSceneScriptingFloatChannel, keys [[1,1],[10,2.2],[11,1]]}
          Shear.X {type=MovieSceneScriptingFloatChannel, const [0,1]}
          Shear.Y {type=MovieSceneScriptingFloatChannel, const [0,1]}
         ]
        }]
      }
      {
       type=MovieSceneColorTrack
       display_name=Color and Opacity
       sections [{
         type=MovieSceneColorSection
         end_frame=11
         start_frame=0
         channels [
          Color.R {type=MovieSceneScriptingFloatChannel, const [1,2]}
          Color.G {type=MovieSceneScriptingFloatChannel, const [1,3]}
          Color.B {type=MovieSceneScriptingFloatChannel, const [1,3]}
          Color.A {type=MovieSceneScriptingFloatChannel, keys [[0,0],[2,1],[6,0.9],[8,0.6],[9,0.15],[10,0],[11,0]]}
         ]
        }]
      }
     ]
    }]}
 }]
graphs [
 EventGraph {nodes [
   Event OnDeltaChanged() {is=title K2Node_Event}
   > Update Delta Bar()
   n1 = Event PreConstruct()
   > n2 = Branch(Condition=@n1.IsDesignTime)
   Initialize Bar(execute=[@n5.then, @n6.then, @n2.then])
   n3 = Cast To FortGameStateAthena!(execute=@n2.else, Object=GetGameState!() -> (ReturnValue:Game State Base Object Reference)) -> (then, CastFailed, AsFort Game State Athena:Fort Game State Athena Object Reference) {is=title K2Node_DynamicCast}
   > n4 = Branch(Condition=IsLastManStandingCN!(self:Fort Game State Athena Object Reference=@n3) -> (ReturnValue:bool))
   > n5 = SetVisibility(InVisibility=Visible)
   n6 = SetVisibility(execute=@n4.else, InVisibility=Collapsed)
   Event OnMaxValueChanged!() -> (OutputDelegate, then, Value:float) {is=title K2Node_Event}
   > UpdateMaxValue()
   n7 = Event OnDBNOStateChanged!() -> (OutputDelegate, then, IsDBNO:bool) {is=title K2Node_Event}
   > UpdateDBNOState!(execute, self, bIsDBNO:bool=@n7.IsDBNO) -> (then)
   Event OnValueChangedWithReason!() -> (OutputDelegate, then, Value:float, Reason:EFortHitPointModificationReason Enum=Invalid) {is=title K2Node_Event}
   > UpdateCurrentValue()
  ]}
 Initialize Bar {nodes [
   n1 = Get BarType() -> (BarType:EHealthBarType Enum=Health)
   SetStyle!(
    execute
    self:Common Text Object Reference=Get MaxValue() -> (MaxValue:Common Text Object Reference)
    InStyle:Common Text Style Class Reference=Select!(
     Health:Common Text Style Class Reference=/FortUILibrary/Styles/Typography/Foundation/TextStyle-Base-XS-HUD-HealthBarMax.TextStyle-Base-XS-HUD-HealthBarMax_C
     Shield:Common Text Style Class Reference=/FortUILibrary/Styles/Typography/Foundation/TextStyle-Base-XS-HUD-ShieldBarMax.TextStyle-Base-XS-HUD-ShieldBarMax_C
     Overshield:Common Text Style Class Reference
     Armor:Common Text Style Class Reference
     Stamina:Common Text Style Class Reference=/FortUILibrary/Styles/Typography/Foundation/TextStyle-Base-XS-HUD-StaminaBarMax.TextStyle-Base-XS-HUD-StaminaBarMax_C
     VehicleHealth:Common Text Style Class Reference=&TextStyle-Base-XS-HUD-StaminaBarMax_C
     SignalInStorm:Common Text Style Class Reference
     Index:EHealthBarType Enum=@n1
    ) -> (ReturnValue:Common Text Style Class Reference)
   ) -> (then)
   > n2 = SetStyle(
    self=Get CurrentValue() -> (CurrentValue:Common Text Object Reference)
    InStyle=/FortUILibrary/Styles/Typography/Foundation/TextStyle-Base-XS-B.TextStyle-Base-XS-B_C
   )
   SetPadding!(
    execute=[@n2.then, @n8.then]
    self:Overlay Slot Object Reference=SlotAsOverlaySlot(Widget=Get TextArea() -> (TextArea:Horizontal Box Object Reference))
    InPadding:Margin Structure=Make Margin(Left=16, Top=0, Right=0, Bottom=0)
   ) -> (then)
   n3 = Select(
    Health:Linear Color Structure=(R=0.0745,G=0.502,B=0.0353,A=1)
    Shield:Linear Color Structure=(R=0.0314,G=0.1804,B=0.8314,A=1)
    Overshield:Linear Color Structure
    Armor:Linear Color Structure
    Stamina:Linear Color Structure=(R=0.6667,G=0.2471,B=0.0275,A=1)
    VehicleHealth:Linear Color Structure=(R=0.5469,G=0.2173,B=0,A=1)
    SignalInStorm:Linear Color Structure=(R=0.0222,G=0.3372,B=0.2623,A=1)
    Index=@n1
   ) -> (ReturnValue:Linear Color Structure)
   n4 = RGB to HSV!(InColor:Linear Color Structure=@n3) -> (H:float, S:float, V:float, A:float)
   n5 = SetContentColorAndOpacity!(
    execute
    self:Border Object Reference=Get CommonBorder_Max() -> (CommonBorder_Max:Common Border Object Reference)
    InContentColorAndOpacity:Linear Color Structure=HSV to RGB!(H:float=@n4, S:float=@n4.S, V:float=float * float(A=@n4.V, B=1), A:float=@n4.A) -> (ReturnValue:Linear Color Structure)
   ) -> (then)
   n6 = Get shieldColor() -> (shieldColor:Linear Color Structure)
   n7 = Get ProgressBarDelta() -> (ProgressBarDelta:Simple Material Progress Bar Object Reference)
   SetColorA!(
    execute=[@n5.then, @n10.then]
    self:Fort Simple Material Progress Bar Object Reference=@n7
    ColorA:Linear Color Structure=Select(
     Health=(R=0.7952,G=0.9,B=0.882,A=1)
     Shield=@n6
     Stamina=(R=0.9137,G=0.7608,B=0.149,A=1)
     VehicleHealth=(R=1,G=0.4852,B=0.1458,A=1)
     SignalInStorm=(R=1,G=1,B=1,A=1)
     Index=@n1
    )
   ) -> (then)
   > SetColorB!(
    execute
    self:Fort Simple Material Progress Bar Object Reference=@n7
    ColorB:Linear Color Structure=Select(
     Health=(R=0.8867,G=0.8901,B=0.9,A=1)
     Shield=@n6
     Stamina=(R=0.807,G=0.491,B=0.0865,A=1)
     VehicleHealth=(R=0.849,G=0.3736,B=0.0486,A=1)
     SignalInStorm=(R=1,G=1,B=1,A=1)
     Index=@n1
    )
   ) -> (then)
   > n8 = SetBrushFromAtlasInterface!(
    execute
    self:Image Object Reference=Get Icon() -> (Icon:Image Object Reference)
    AtlasRegion:Slate Texture Atlas Interface Interface=Select(
     Health:Slate Texture Atlas Interface Interface=Get HealthIcon() -> (HealthIcon:Paper Sprite Object Reference)
     Shield:Slate Texture Atlas Interface Interface=Get ShieldIcon() -> (ShieldIcon:Paper Sprite Object Reference)
     Overshield:Slate Texture Atlas Interface Interface
     Armor:Slate Texture Atlas Interface Interface
     Stamina:Slate Texture Atlas Interface Interface
     VehicleHealth:Slate Texture Atlas Interface Interface=Get VehicleHealthIcon() -> (VehicleHealthIcon:Paper Sprite Object Reference)
     SignalInStorm:Slate Texture Atlas Interface Interface
     Index=@n1
    ) -> (ReturnValue:Slate Texture Atlas Interface Interface)
    bMatchSize:bool=false
   ) -> (then)
   Initialize Bar() {is=title K2Node_FunctionEntry}
   n9 = Get ProgressBarFill() -> (ProgressBarFill:Simple Material Progress Bar Object Reference)
   > SetColorA(self=@n9, ColorA=@n3)
   > n10 = SetColorB(
    self=@n9
    ColorB=Select(
     Health:Linear Color Structure=(R=0.1765,G=0.7294,B=0.0784,A=1)
     Shield:Linear Color Structure=(R=0.0863,G=0.4392,B=0.9137,A=1)
     Overshield:Linear Color Structure
     Armor:Linear Color Structure
     Stamina:Linear Color Structure=(R=0.8235,G=0.5529,B=0.0627,A=1)
     VehicleHealth:Linear Color Structure=(R=0.7604,G=0.3523,B=0.0832,A=1)
     SignalInStorm:Linear Color Structure=(R=0.0222,G=0.3372,B=0.2623,A=1)
     Index=@n1
    ) -> (ReturnValue:Linear Color Structure)
   )
  ]}
 SetSize {nodes [
   n1 = SetSize!() -> (then, useLargeSize:bool=~) {is=title K2Node_FunctionEntry}
   n2 = Select!(Option 0:double=Get NormalHeight() -> (NormalHeight:double), Option 1:double=Get TallHeight() -> (TallHeight:double), Index:bool=@n1) -> (ReturnValue:double)
   > SetHeightOverride(self=Get BarSizeBox() -> (BarSizeBox:Size Box Object Reference), InHeightOverride=@n2)
   > SetDesiredSizeOverride!(execute, self:Image Object Reference=Get Icon(), DesiredSize:Vector 2D Structure=MakeVector2D(X=@n2, Y=@n2)) -> (then)
  ]}
 Update Delta Bar {nodes [
   Update Delta Bar() {is=title K2Node_FunctionEntry}
   > SetProgress!(
    execute
    self:Fort Simple Material Progress Bar Object Reference=Get ProgressBarDelta()
    Progress:float=GetLastValuePercentage!(self:Athena Player Hit Point Bar Base Object Reference) -> (ReturnValue:float)
   ) -> (then)
  ]}
 Update Fill Bar {nodes [
   Update Fill Bar() {is=title K2Node_FunctionEntry}
   > SetProgress(
    self=Get ProgressBarFill()
    Progress=GetCurrentValuePercentage!(self:Athena Player Hit Point Bar Base Object Reference) -> (ReturnValue:float)
   )
  ]}
 UpdateCurrentValue {nodes [
   UpdateCurrentValue() {is=title K2Node_FunctionEntry}
   n1 = GetCurrentValuePercentage()
   > n2 = Branch(Condition="float > float"(A=@n1, B=0.99) {is=title K2Node_CallFunction})
   > n3 = Set IndexNodamage(IndexNodamage:int=1) -> (Output_Get:int=0)
   n4 = float - float(A=float / float(A=@n1, B=0.25) {is=title K2Node_CallFunction}, B=1) {is=title K2Node_CallFunction}
   n5 = Set Index(
    execute=[@n11.then, @n3.then]
    Index:int="Clamp (Integer)"(Value=Ceil(A=@n4) {is=title K2Node_CallFunction}, Min=0, Max=3) -> (ReturnValue=0)
   ) -> (Output_Get:int=0)
   n6 = Get status() -> (status:Array of Texts)
   n7 = Get status()
   n8 = Get PlayerState() -> (PlayerState:Common Text Object Reference)
   > "SetText (Text)"(
    self=@n8
    InText=Select(
     Option 0:Text="Get (a copy)"(Array:Array of Texts=@n7, Dimension 1=@n5) -> (Output:Text)
     Option 1:Text="Get (a copy)"(Array=@n6, Dimension 1=4)
     Index:int=Get IndexNodamage() -> (IndexNodamage:int=0)
    ) -> (ReturnValue:Text)
   )
   > Update Fill Bar() {is=title K2Node_CallFunction}
   n9 = "Get (a copy)"(Array:Array of Linear Color Structures=Get TextColor() -> (TextColor:Array of Linear Color Structures), Dimension 1=@n5) -> (Output:Linear Color Structure)
   > SetColorAndOpacity(self=@n8, InColorAndOpacity=Make Slate Color(SpecifiedColor=@n9))
   n10 = Get ProgressBarFill()
   > SetColorA(self=@n10, ColorA="Get (a copy)"(Array=Get StormColorAA() -> (StormColorAA:Array of Linear Color Structures), Dimension 1=@n5))
   > SetColorB(self=@n10, ColorB="Get (a copy)"(Array=Get StormColorBB() -> (StormColorBB:Array of Linear Color Structures), Dimension 1=@n5))
   > updateAnimation()
   > updateEdgeColor!(execute, self, NewParam:Linear Color Structure=@n9) -> (then)
   n11 = Set IndexNodamage(execute=@n2.else, IndexNodamage=0) -> (Output_Get=0)
   n12 = Get ValueCurrent() -> (ValueCurrent:float)
   n13 = Round(A=@n12) {is=title K2Node_CallFunction}
   n14 = Branch(Condition=AND Boolean(A="Equal (Integer)"(A=@n13, B=0) {is=title K2Node_CallFunction}, B="float > float"(A=@n12, B=0)))
   n15 = Get CurrentValue()
   > "SetText (Text)"(self=@n15, InText="To Text (Integer)"(Value=1))
   "SetText (Text)"(execute=@n14.else, self=@n15, InText="To Text (Integer)"(Value=@n13))
  ]}
 UpdateDBNOState {nodes [
   n1 = UpdateDBNOState!() -> (then, bIsDBNO:bool=~) {is=title K2Node_FunctionEntry}
   > n2 = Branch(Condition="Equal (Boolean)"!(A:bool=@n1, B:bool=Get DBNOStateCache() -> (DBNOStateCache:bool)) -> (ReturnValue:bool))
   n3 = Set DBNOStateCache(execute=@n2.else, DBNOStateCache:bool=@n1) -> (Output_Get:bool)
   > n4 = Switch on EHealthBarType!(execute, Selection:EHealthBarType Enum=Get BarType() -> (BarType=Health)) -> (Health:Exec, Shield:Exec, Overshield:Exec, Armor:Exec, Stamina:Exec, VehicleHealth:Exec, SignalInStorm:Exec) {is=title K2Node_SwitchEnum}
   > n5 = Branch(Condition=@n3)
   > SetProgress(self=Get ProgressBarDelta(), Progress=0)
   > SetBrushFromTexture!(
    execute
    self:Image Object Reference=Get Icon()
    Texture:Texture 2D Object Reference=/Game/Athena/HUD/Art/squadArt/squad_Eliminated
    bMatchSize:bool=false
   ) -> (then)
   n6 = Get ProgressBarFill()
   > SetColorA(self=@n6, ColorA=(R=0.5776,G=0,B=0.0232,A=1))
   > SetColorB(self=@n6, ColorB=(R=1,G=0,B=0.1144,A=1))
   > SetStyle(self=Get MaxValue(), InStyle=/FortUILibrary/Styles/Typography/Foundation/TextStyle-Base-XS-HUD-DBNOBarMax.TextStyle-Base-XS-HUD-DBNOBarMax_C)
   Initialize Bar(execute=@n5.else) {is=title K2Node_CallFunction}
   n7 = Branch(execute=[@n4.Shield, @n4.SignalInStorm], Condition=@n3)
   > SetColorAndOpacity(self:User Widget Object Reference, InColorAndOpacity:Linear Color Structure=(R=1,G=1,B=1,A=0))
   SetColorAndOpacity(execute=@n7.else, InColorAndOpacity=(R=1,G=1,B=1,A=1))
  ]}
 UpdateMaxValue {nodes [
   UpdateMaxValue() {is=title K2Node_FunctionEntry}
   > Update Fill Bar()
   > Update Delta Bar() {is=title K2Node_CallFunction}
   > "SetText (Text)"(
    self=Get MaxValue()
    InText="To Text (Float)"(
     Value=Get ValueMax() -> (ValueMax:float)
     RoundingMode=FromZero
     MinimumIntegralDigits=1
     MaximumIntegralDigits=324
     MinimumFractionalDigits=0
     MaximumFractionalDigits=0
    )
   )
  ]}
 setGlow {nodes [
   setGlow() {is=title K2Node_FunctionEntry}
   > Set outlineGlow(outlineGlow:bool=false) -> (Output_Get:bool)
  ]}
 updateAnimation {nodes [
   n1 = Get warning() -> (warning:Widget Animation Object Reference)
   n2 = StopAnimation(InAnimation=@n1)
   PlayAnimation(execute=[@n2.then, @n8.then], InAnimation=@n1, PlaybackSpeed=0.55)
   SetVisibility(self=Get BorderEdge3() -> (BorderEdge3:Image Object Reference), InVisibility=Visible)
   > Set outlineGlow(outlineGlow=false)
   n3 = Get warning()
   > StopAnimation(InAnimation=@n3)
   > PlayAnimation(PlaybackSpeed=0.55)
   updateAnimation() {is=title K2Node_FunctionEntry}
   n4 = GetCurrentValuePercentage()
   n5 = GetLastValuePercentage()
   n6 = float - float(A=@n5, B=@n4)
   > n7 = Branch(Condition="float < float"!(A:double=@n6, B:double=0) -> (ReturnValue:bool) {is=title K2Node_CallFunction})
   > Set outlineGlow(outlineGlow=false)
   > n8 = SetVisibility(self=Get BorderEdge3(), InVisibility=Hidden)
   n9 = Branch(execute=@n7.else, Condition="float <= float"!(A:double=@n4, B:double=0.25) -> (ReturnValue:bool) {is=title K2Node_CallFunction})
   > SetVisibility(self=Get BorderEdge3(), InVisibility=Visible)
   > Set outlineGlow(outlineGlow=true)
   > PlayAnimation(InAnimation=@n3, PlaybackSpeed=0.55)
   SetVisibility(execute=@n9.else, self=Get BorderEdge3(), InVisibility=Visible)
   n10 = float - float(A=float / float(A=@n5, B=0.25), B=1)
   > n11 = Set IndexState(IndexState:int="Clamp (Integer)"(Value=Ceil(A=@n10), Min=0, Max=3) -> (ReturnValue=0)) -> (Output_Get:int=0)
   n12 = int - int(A=@n11, B=Get Index() -> (Index:int=0)) {is=title K2Node_CallFunction}
   > n13 = Branch(Condition="integer >= integer"!(A:int=@n12, B:int=0) -> (ReturnValue:bool) {is=title K2Node_CallFunction})
   > Branch(Condition="Equal (Integer)"(A=@n12, B=1))
   > n14 = Branch(Condition="Equal (Integer)"(A=@n11, B=1))
   n15 = Get warning()
   StopAnimation(execute=@n14.else, InAnimation=@n15)
   > PlayAnimation(InAnimation=@n15, NumLoopsToPlay=3, PlaybackSpeed=0.55)
   Branch(execute=@n13.else, Condition="Equal (Integer)"(A=@n11, B=0))
   > StopAnimation(InAnimation=@n15)
   > PlayAnimation(InAnimation=@n15, NumLoopsToPlay=3, PlaybackSpeed=0.55)
   Set outlineGlow(outlineGlow=true)
   > PlayAnimation(InAnimation=@n15, NumLoopsToPlay=0, PlaybackSpeed=0.55)
   "float <= float"(A=@n6, B=0)
   BindToAnimationFinished(Delegate=Create Event())
  ]}
 updateEdgeColor {nodes [
   n1 = updateEdgeColor!() -> (then, NewParam:Linear Color Structure) {is=title K2Node_FunctionEntry}
   > n2 = Branch(Condition=Get outlineGlow() -> (outlineGlow:bool))
   > n3 = For Loop(LastIndex=1)
   n4 = Get BorderEdge4() -> (BorderEdge4:Image Object Reference)
   n5 = Get BorderEdge3()
   > n6 = GetDynamicMaterial(self="Get (a copy)"(
     Array:Array of Image Object References=Make Array!("[0]":Image Object Reference=@n4, "[1]":Image Object Reference=@n5) -> (Array:Array of Image Object References)
     Dimension 1=@n3
    ) -> (Output:Image Object Reference))
   > SetVectorParameterValue(self=@n6, ParameterName=Glow, Value=(R=1,G=1,B=1,A=1))
   > SetVectorParameterValue(self=@n6, ParameterName=EdgeColor, Value=@n1)
   n7 = For Loop(execute=@n2.else, LastIndex=1)
   n8 = Get BorderEdge4()
   n9 = Get BorderEdge3()
   > n10 = GetDynamicMaterial(self="Get (a copy)"(Array=Make Array("[0]"=@n8, "[1]"=@n9), Dimension 1=@n7))
   > SetVectorParameterValue(self=@n10, ParameterName=Glow, Value=(R=0,G=0,B=0,A=1))
   > SetVectorParameterValue(self=@n10, ParameterName=EdgeColor, Value=@n1)
   Get BorderEdge5() -> (BorderEdge5:Image Object Reference)
   Get BorderEdge6() -> (BorderEdge6:Image Object Reference)
   Get BorderEdge7() -> (BorderEdge7:Image Object Reference)
  ]}
]
variables [
 DBNOStateCache {type=boolean}
 HealthIcon {type=PaperSprite, default=/Game/Athena/HUD/Atlas/Shared/icon_health_Sprite}
 Index {type=integer}
 IndexNodamage {type=integer}
 IndexState {type=integer, default=3}
 outlineGlow {type=boolean}
 shieldColor {type=LinearColor, default=<rgba 0.6351 0.6716 0.8314 1>}
 ShieldIcon {type=PaperSprite, default=/Game/Athena/HUD/Atlas/Shared/icon_shield_Sprite}
 status {type="array<string>", default [x2.0,x1.6,x1.3,x1.1,x1.0]}
 StormColorAA {
  type="array<LinearColor>"
  default [<rgba 0.6795 0.0272 0.0272 1>, <rgba 0.6514 0.1248 0.0723 1>, <rgba 0.807 0.5395 0.1221 1>, <rgba 0.0222 0.3372 0.2623 1>]
 }
 StormColorBB {
  type="array<LinearColor>"
  default [<rgba 0.6795 0.0382 0.0382 1>, <rgba 0.568 0.1088 0.063 1>, <rgba 0.671 0.4486 0.1016 1>, <rgba 0.0343 0.5029 0.3916 1>]
 }
 TextColor {
  type="array<LinearColor>"
  default [<rgba 0.946 0.0378 0.0378 1>, <rgba 0.798 0.1529 0.0885 1>, <rgba 0.974 0.6512 0.1474 1>, <rgba 0.1405 0.895 0.7158 1>]
 }
 VehicleHealthIcon {type=PaperSprite, default=/Game/Athena/HUD/Atlas/Shared/icon_vehiclehealth_sprite.icon_vehiclehealth_Sprite}
]
widgets {
 info {
  inheritedWidgetCount=1
  parentClass=/Script/FortniteUI.AthenaPlayerHitPointBarBase
  rootWidgetClass=&HorizontalBox
  widgetCount=27
 }
 widgets [
  HorizontalBox_66 {widgetClassPath=&HorizontalBox, widget=HorizontalBox_66}
  Icon {
   widgetClassPath=&Image
   bIsVariable=true
   parent=HorizontalBox_66
   widget=Icon
   properties {
    brush {
     bIsDynamicallyLoaded=false
     drawAs=Image
     imageType=NoImage
     mirroring=NoMirror
     resourceName=None
     resourceObject=null
     tiling=NoTile
     imageSize=<24 24>
     margin=<ltrb 0 0 0 0>
     outlineSettings {
      bUseBrushTransparency=false
      roundingType=HalfHeightRadius
      width=0
      color {colorUseRule=UseColor_Specified, specifiedColor=<rgba 0 0 0 0>}
      cornerRadii=<0 0 0 0>
     }
     tintColor {colorUseRule=UseColor_Specified, specifiedColor=<rgba 1 1 1 0>}
     uVRegion {bIsValid=false, max=<0 0>, min=<0 0>}
    }
    renderTransform {angle=0, scale=<1 1>, shear=<12 0>, translation=<0 0>}
   }
   slot {class=HorizontalBoxSlot, properties {horizontalAlignment=HAlign_Center, verticalAlignment=VAlign_Center, padding=<ltrb -12 -12 0 -12>}}
  }
  HorizontalBox_1 {
   widgetClassPath=&HorizontalBox
   parent=HorizontalBox_66
   widget=HorizontalBox_1
   slot {class=HorizontalBoxSlot, properties {size {sizeRule=Fill, value=1}}}
  }
  signal_icon {
   widgetClassPath=&SizeBox
   bIsVariable=true
   parent=HorizontalBox_1
   widget=signal_icon
   properties {bOverride_HeightOverride=true, bOverride_WidthOverride=true, heightOverride=24, widthOverride=105}
   slot {class=HorizontalBoxSlot, properties {horizontalAlignment=HAlign_Left}}
  }
  Overlay {widgetClassPath=&Overlay, parent=signal_icon, widget=Overlay, properties {visibility=HitTestInvisible}, slot {class=SizeBoxSlot}}
  SignalBG {
   widgetClassPath=&Image
   bIsVariable=true
   parent=Overlay
   widget=SignalBG
   properties {brush {
     bIsDynamicallyLoaded=false
     drawAs=Image
     imageType=NoImage
     mirroring=NoMirror
     resourceName=None
     resourceObject=null
     tiling=NoTile
     imageSize=<32 32>
     margin=<ltrb 0 0 0 0>
     outlineSettings {
      bUseBrushTransparency=false
      roundingType=HalfHeightRadius
      width=0
      color {colorUseRule=UseColor_Specified, specifiedColor=<rgba 0 0 0 0>}
      cornerRadii=<0 0 0 0>
     }
     tintColor {colorUseRule=UseColor_Specified, specifiedColor=<rgba 0.0026 0.005 0 0.8>}
     uVRegion {bIsValid=false, max=<0 0>, min=<0 0>}
    }}
   slot {class=OverlaySlot, properties {horizontalAlignment=HAlign_Fill, verticalAlignment=VAlign_Fill, padding=<ltrb 1 1 1 1>}}
  }
  BorderEdge {
   widgetClassPath=&Image
   bIsVariable=true
   parent=Overlay
   widget=BorderEdge
   properties {brush {
     bIsDynamicallyLoaded=false
     drawAs=Image
     imageType=NoImage
     mirroring=NoMirror
     resourceName=None
     resourceObject=&/T_UI_signal_edge2
     tiling=NoTile
     imageSize=<105 24>
     margin=<ltrb 0 0 0 0>
     outlineSettings {
      bUseBrushTransparency=false
      roundingType=HalfHeightRadius
      width=0
      color {colorUseRule=UseColor_Specified, specifiedColor=<rgba 0 0 0 0>}
      cornerRadii=<0 0 0 0>
     }
     tintColor {colorUseRule=UseColor_Specified, specifiedColor=<rgba 1 1 1 0.8>}
     uVRegion {bIsValid=false, max=<0 0>, min=<0 0>}
    }}
   slot {class=OverlaySlot, properties {horizontalAlignment=HAlign_Fill, verticalAlignment=VAlign_Fill, padding=<ltrb 1 1 1 1>}}
  }
  HorizontalBox_44 {
   widgetClassPath=&HorizontalBox
   parent=Overlay
   widget=HorizontalBox_44
   slot {class=OverlaySlot, properties {horizontalAlignment=HAlign_Fill, verticalAlignment=VAlign_Fill}}
  }
  SignalIcon {
   widgetClassPath=&Image
   bIsVariable=true
   parent=HorizontalBox_44
   widget=SignalIcon
   properties {brush {
     bIsDynamicallyLoaded=false
     drawAs=Image
     imageType=NoImage
     mirroring=NoMirror
     resourceName=None
     resourceObject=&/T_UI_signalIcon
     tiling=NoTile
     imageSize=<22 22>
     margin=<ltrb 0 0 0 0>
     outlineSettings {
      bUseBrushTransparency=false
      roundingType=HalfHeightRadius
      width=0
      color {colorUseRule=UseColor_Specified, specifiedColor=<rgba 0 0 0 0>}
      cornerRadii=<0 0 0 0>
     }
     tintColor {colorUseRule=UseColor_Specified, specifiedColor=<rgba 1 1 1 1>}
     uVRegion {bIsValid=false, max=<0 0>, min=<0 0>}
    }}
   slot {class=HorizontalBoxSlot, properties {horizontalAlignment=HAlign_Center, verticalAlignment=VAlign_Bottom, padding=<ltrb 10 0 0 1>}}
  }
  PlayerState {
   widgetClassPath=&CommonTextBlock
   bIsVariable=true
   parent=HorizontalBox_44
   widget=PlayerState
   properties {
    justification=Center
    text=x1.0
    colorAndOpacity {colorUseRule=UseColor_Specified, specifiedColor=<rgba 0.1052 0.67 0.5358 1>}
    font {
     bForceMonospaced=false
     bMaterialIsStencil=false
     fontMaterial=null
     fontObject=&BurbankSmall
     letterSpacing=0
     monospacedWidth=1
     size=14
     skewAmount=0
     typefaceFontName=Bold
     outlineSettings {
      bApplyOutlineToDropShadows=false
      bMiteredCorners=false
      bSeparateFillAlpha=false
      outlineMaterial=null
      outlineSize=0
      outlineColor=<rgba 0 0 0 1>
     }
    }
    renderTransform {angle=0, scale=<0.9 0.9>, shear=<0 0>, translation=<0 0>}
    shadowOffset=<0 0>
   }
   slot {class=HorizontalBoxSlot, properties {horizontalAlignment=HAlign_Left, verticalAlignment=VAlign_Center, padding=<ltrb 10 0 0 0>}}
  }
  BarSizeBox {
   widgetClassPath=&SizeBox
   bIsVariable=true
   parent=HorizontalBox_1
   widget=BarSizeBox
   properties {heightOverride=22, widthOverride=410}
   slot {class=HorizontalBoxSlot, properties {size {sizeRule=Fill, value=1}}}
  }
  "[Overlay]" {
   widgetClassPath=&Overlay
   parent=BarSizeBox
   widget="[Overlay]"
   properties {visibility=HitTestInvisible, renderTransform {angle=0, scale=<1 1>, shear=<0 0>, translation=<0 1>}}
   slot {class=SizeBoxSlot, properties {verticalAlignment=VAlign_Center}}
  }
  BorderEdge3 {
   widgetClassPath=&Image
   bIsVariable=true
   parent="[Overlay]"
   widget=BorderEdge3
   properties {
    brush {
     bIsDynamicallyLoaded=false
     drawAs=Image
     imageType=NoImage
     mirroring=NoMirror
     resourceName=None
     resourceObject=&/MI_UI_signal_edge2
     tiling=NoTile
     imageSize=<410 16>
     margin=<ltrb 0.578 1 0.7615 1>
     outlineSettings {
      bUseBrushTransparency=false
      roundingType=HalfHeightRadius
      width=0
      color {colorUseRule=UseColor_Specified, specifiedColor=<rgba 0 0 0 0>}
      cornerRadii=<0 0 0 0>
     }
     tintColor {colorUseRule=UseColor_Specified, specifiedColor=<rgba 1 1 1 1>}
     uVRegion {bIsValid=false, max=<0 0>, min=<0 0>}
    }
    colorAndOpacity=<rgba 1 1 1 0>
   }
   slot {class=OverlaySlot, properties {verticalAlignment=VAlign_Bottom, padding=<ltrb 2 0 0 1>}}
  }
  SizeBox_1 {
   widgetClassPath=&SizeBox
   parent="[Overlay]"
   widget=SizeBox_1
   properties {bOverride_HeightOverride=true, heightOverride=12}
   slot {class=OverlaySlot, properties {horizontalAlignment=HAlign_Fill, verticalAlignment=VAlign_Bottom}}
  }
  Overlay_0 {widgetClassPath=&Overlay, parent=SizeBox_1, widget=Overlay_0, slot {class=SizeBoxSlot}}
  ProgressBarbg {
   widgetClassPath=/Game/Athena/HUD/SimpleMaterialProgressBar
   bIsVariable=true
   parent=Overlay_0
   widget=ProgressBarbg
   properties {
    bOverrideDefaultColorA=true
    bOverrideDefaultColorB=true
    bOverrideDefaultColorBackground=true
    bOverrideDefaultSegmentEdge=true
    bOverrideDefaultSegments=true
    segmentEdge=0.0004
    segments=4
    cachedColorA=<rgba 0.0048 0.0048 0.0048 0.8>
    cachedColorB=<rgba 0.0048 0.0048 0.0048 0.8>
    cachedColorBackground=<rgba 0.0048 0.0048 0.0048 0.8>
    renderTransform {angle=0, scale=<1 1.3>, shear=<0 0>, translation=<0 -3>}
   }
   slot {class=OverlaySlot, properties {horizontalAlignment=HAlign_Fill, verticalAlignment=VAlign_Bottom}}
  }
  ProgressBarDelta {
   widgetClassPath=&SimpleMaterialProgressBar
   bIsVariable=true
   parent=Overlay_0
   widget=ProgressBarDelta
   properties {
    bOverrideDefaultColorBackground=true
    bOverrideDefaultSegmentEdge=true
    bOverrideDefaultSegments=true
    designTime_Progress=0.75
    segmentEdge=0.0004
    segments=4
    cachedColorBackground=<rgba 0 0.0012 0.0103 0>
    renderTransform {angle=0, scale=<1 1.3>, shear=<0 0>, translation=<0 -3>}
   }
   slot {class=OverlaySlot, properties {horizontalAlignment=HAlign_Fill, verticalAlignment=VAlign_Bottom}}
  }
  ProgressBarFill {
   widgetClassPath=&SimpleMaterialProgressBar
   bIsVariable=true
   parent=Overlay_0
   widget=ProgressBarFill
   properties {
    bOverrideDefaultGlowEdge=true
    bOverrideDefaultGlowSoftness=true
    bOverrideDefaultOutlineScale=true
    bOverrideDefaultSegmentEdge=true
    bOverrideDefaultSegments=true
    designTime_Progress=0.5
    segmentEdge=0.0004
    segments=4
    renderTransform {angle=0, scale=<1 1.3>, shear=<0 0>, translation=<0 -3>}
   }
   slot {class=OverlaySlot, properties {horizontalAlignment=HAlign_Fill, verticalAlignment=VAlign_Bottom}}
  }
  BorderEdge4 {
   widgetClassPath=&Image
   bIsVariable=true
   parent="[Overlay]"
   widget=BorderEdge4
   properties {brush {
     bIsDynamicallyLoaded=false
     drawAs=Image
     imageType=NoImage
     mirroring=NoMirror
     resourceName=None
     resourceObject=&/MI_UI_signal_edge
     tiling=NoTile
     imageSize=<100 16>
     margin=<ltrb 0 0 0 0>
     outlineSettings {
      bUseBrushTransparency=false
      roundingType=HalfHeightRadius
      width=0
      color {colorUseRule=UseColor_Specified, specifiedColor=<rgba 0 0 0 0>}
      cornerRadii=<0 0 0 0>
     }
     tintColor {colorUseRule=UseColor_Specified, specifiedColor=<rgba 1 1 1 1>}
     uVRegion {bIsValid=false, max=<0 0>, min=<0 0>}
    }}
   slot {class=OverlaySlot, properties {horizontalAlignment=HAlign_Fill}}
  }
  SizeBox_0 {
   widgetClassPath=&SizeBox
   parent=HorizontalBox_1
   widget=SizeBox_0
   properties {bOverride_WidthOverride=true, visibility=Hidden, widthOverride=200}
   slot {class=HorizontalBoxSlot, properties {horizontalAlignment=HAlign_Left, verticalAlignment=VAlign_Center, padding=<ltrb 0 0 -200 -2>}}
  }
  TextArea {
   widgetClassPath=&HorizontalBox
   bIsVariable=true
   parent=SizeBox_0
   widget=TextArea
   slot {class=SizeBoxSlot, properties {padding=<ltrb 8 0 0 0>}}
  }
  CurrentValue {
   widgetClassPath=&CommonTextBlock
   bIsVariable=true
   parent=TextArea
   widget=CurrentValue
   properties {
    justification=Right
    text=999?
    font {
     bForceMonospaced=false
     bMaterialIsStencil=false
     fontMaterial=null
     fontObject=&Burbank
     letterSpacing=0
     monospacedWidth=1
     size=16
     skewAmount=0
     typefaceFontName=Burbank-PlayerPower
     outlineSettings {
      bApplyOutlineToDropShadows=false
      bMiteredCorners=false
      bSeparateFillAlpha=false
      outlineMaterial=null
      outlineSize=0
      outlineColor=<rgba 0 0 0 1>
     }
    }
    renderTransform {angle=0, scale=<1 1>, shear=<0 0>, translation=<0 -2>}
    shadowOffset=<0 0>
   }
   slot {class=HorizontalBoxSlot, properties {horizontalAlignment=HAlign_Left, verticalAlignment=VAlign_Center}}
  }
  HorizontalBox_MaxValue {
   widgetClassPath=&HorizontalBox
   bIsVariable=true
   parent=TextArea
   widget=HorizontalBox_MaxValue
   properties {visibility=Collapsed}
   slot {class=HorizontalBoxSlot, properties {horizontalAlignment=HAlign_Left}}
  }
  Split {
   widgetClassPath=&Image
   parent=HorizontalBox_MaxValue
   widget=Split
   properties {
    brush {
     bIsDynamicallyLoaded=false
     drawAs=Image
     imageType=NoImage
     mirroring=NoMirror
     resourceName=None
     resourceObject=null
     tiling=NoTile
     imageSize=<2 19.455>
     margin=<ltrb 0 0 0 0>
     outlineSettings {
      bUseBrushTransparency=false
      roundingType=HalfHeightRadius
      width=0
      color {colorUseRule=UseColor_Specified, specifiedColor=<rgba 0 0 0 0>}
      cornerRadii=<0 0 0 0>
     }
     tintColor {colorUseRule=UseColor_Specified, specifiedColor=<rgba 1 1 1 1>}
     uVRegion {bIsValid=false, max=<0 0>, min=<0 0>}
    }
    colorAndOpacity=<rgba 1 1 1 0.2>
    renderTransform {angle=0, scale=<1 1>, shear=<-10 0>, translation=<0 -2>}
   }
   slot {class=HorizontalBoxSlot, properties {horizontalAlignment=HAlign_Left, verticalAlignment=VAlign_Center, padding=<ltrb 8 0 8 0>}}
  }
  CommonBorder_Max {
   widgetClassPath=&CommonBorder
   bIsVariable=true
   parent=HorizontalBox_MaxValue
   widget=CommonBorder_Max
   properties {horizontalAlignment=HAlign_Left, padding=<ltrb 0 0 0 0>}
   slot {class=HorizontalBoxSlot, properties {horizontalAlignment=HAlign_Left, verticalAlignment=VAlign_Center}}
  }
  MaxValue {
   widgetClassPath=&CommonTextBlock
   bIsVariable=true
   parent=CommonBorder_Max
   widget=MaxValue
   properties {
    text=999?
    font {
     bForceMonospaced=false
     bMaterialIsStencil=false
     fontMaterial=null
     fontObject=&Burbank
     letterSpacing=0
     monospacedWidth=1
     size=16
     skewAmount=0
     typefaceFontName=Burbank-PlayerPower
     outlineSettings {
      bApplyOutlineToDropShadows=false
      bMiteredCorners=false
      bSeparateFillAlpha=false
      outlineMaterial=null
      outlineSize=0
      outlineColor=<rgba 0 0 0 1>
     }
    }
    renderTransform {angle=0, scale=<1 1>, shear=<0 0>, translation=<0 -2>}
    shadowOffset=<0 0>
   }
   slot {class=BorderSlot, properties {horizontalAlignment=HAlign_Left, padding=<ltrb 0 0 0 0>}}
  }
  TutorialHighlight {widgetClassPath=&TutorialHighlight}
 ]
}
